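# Niagara system for the RollerBlade emote that adds a warm gold glow and sparkles to a skeletal mesh.
# One-shot bursts create flares, a corona, a brief light, mesh glow and GPU sparks, while glints, drifting sparkles and a periodic mesh sheen repeat.
# Emitters sample the skeletal mesh for placement, with glints partitioned across three socket or bone indices.
# Time-of-day scalars tone down several layers, and some emitters use distance culling.
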NiagaraSystem /BRCosmetics/Animation/Game/MainPlayer/Emotes/RollerBlade/FX/NS_Emote_RollerBlade
dependencies {
 usedDataInterfaces [&NiagaraDataInterfaceArrayColor, &NiagaraDataInterfaceCurve, &NiagaraDataInterfaceColorCurve]
 usedDynamicInputs []
 usedModules [
  &self:SystemSpawnScript.SystemScriptSource.SystemScriptGraph.NiagaraNodeAssignment_0.SetVariables_#1
  &SystemState
 ]
}
emitters {
 BG_Flare {
  emitterSpawnScript [/SetVariables(Emitter.RandomPerSpawn:NiagaraFloat=&MakeCustomFloatFromBool(
     Bool=&RandomBool(Randomness Mode=::NewEnumerator2, Random Seed=?, Fixed Random Seed=true)
     True Float=0
     False Float=1
    )) {isSetParameters=true}]
  emitterUpdateScript [
   &EmitterState(
    Enable Distance Culling=true
    Life Cycle Mode=::NewEnumerator1
    Loop Behavior=::NewEnumerator1
    Loop Duration=0
    Max Distance Response=::KillImmediately
    Module.bUseMaxDistance=true
    Module.bUseMinDistance=false
    Scalability Mode=::NewEnumerator1
    UseLoopDelay=false
   )
   &SpawnBurst_Instantaneous(Use Spawn Probability=false, UseLoopCountLimit=false)
  ]
  particleSpawnScript [
   /SetVariables(Particles.RandomPerSpawn:NiagaraFloat=&MakeCustomFloatFromBool(Bool=&RandomBool(Randomness Mode=::NewEnumerator1), True Float=1, False Float=0)) {isSetParameters=true}
   &InitializeParticle(
    Color=&MultiplyLinearColor(
     Linear Color=<rgba 1 0.4667 0 1>
     Scale Factor=&MakeFloat4FromVectorAndFloat("Vector (RGB)"=&VectorFromFloat(Value=0.5), "Float (Alpha)"=0.6)
    )
    Sprite Rotation Mode=::NewEnumerator2
    Sprite Size Mode=::NewEnumerator3
    Uniform Sprite Size=&Multiply_Float(A=&UniformRangedFloat(Minimum=120, Maximum=150), B=1)
    UsePositionOffset=false
    Write Material Random=true
   )
  ]
  particleUpdateScript [
   &ParticleState(Kill Particles When Lifetime Has Elapsed=false)
   &ScaleColor(
    Scale Alpha=&FloatFromCurve(FloatCurve=DataInterface() {propertyValues {Curve {keys [
         {arriveTangent=-2.6103, interpMode=RCIM_Cubic, leaveTangent=-2.6103, tangentMode=RCTM_User, time=0, value=1}
         {interpMode=RCIM_Cubic, leaveTangent=-4, time=0.4, value=0.5}
        ]}}})
    Scale RGB=&VectorFromFloat(Value=&FNTimeOfDayScalar(Night Scale=0.65, Morning Scale=0.4, Evening Scale=0.4))
    ScaleA=true
    ScaleRGB=true
   )
   &SkeletalMeshLocation(Bone Sampling Mode=::NewEnumerator4, MaskSpawnGroup=false, Skeletal Mesh=$System.Skeletal Mesh)
   &CameraOffset(Camera Offset Amount=-10)
  ]
  rendererClasses [&NiagaraSpriteRendererProperties]
  renderers [{rendererClass=&NiagaraSpriteRendererProperties}]
 }
 Corona {
  emitterUpdateScript [
   &EmitterState(Life Cycle Mode=::NewEnumerator1, Loop Behavior=::NewEnumerator1, Loop Duration=2, UseLoopDelay=false)
   &SpawnBurst_Instantaneous(Spawn Count=1, Use Spawn Probability=true)
  ]
  particleSpawnScript [
   &InitializeParticle(
    Color=<rgba 9 3.7221 0 10>
    Mesh Scale Mode=::NewEnumerator3
    Mesh Uniform Scale=&Multiply_Float(A=1.5, B=1)
    UsePositionOffset=false
    Write Material Random=true
   )
   &SkeletalMeshLocation(
    MaskSpawnGroup=false
    Mesh Sampling Type=::NewEnumerator5
    Orientation Sampling=::NewEnumerator1
    Sampled Position Offset=<0 0 -35>
    Sampled Position Offset Coordinate Space:ENiagaraSkelMeshTransforms=::NewEnumerator2
    Skeletal Mesh=$System.Skeletal Mesh
    Socket Sampling Mode=::NewEnumerator5
   )
  ]
  particleUpdateScript [
   &ParticleState(Kill Particles When Lifetime Has Elapsed=false, Let Infinitely Lived Particles Die When Emitter Deactivates=true)
   &DynamicMaterialParameters(
    Index 0 Param 1=1.1
    Index 0 Param 2=0.25
    Index 0 Param 3=1
    Index 0 Param 4=1
    Module.Param0WriteEnabled=true
    Module.Param1WriteEnabled=true
    Module.Param2WriteEnabled=true
    Module.Param3WriteEnabled=true
   )
   &ScaleColor(
    Scale Alpha=&Multiply_Float(
     A=2
     B=&FloatFromCurve(
      FloatCurve=DataInterface() {propertyValues {Curve {keys [
          {arriveTangent=12.8732, interpMode=RCIM_Cubic, leaveTangent=12.8732, tangentMode=RCTM_User, time=-0.0025, value=0.4068}
          {arriveTangent=0.0247, interpMode=RCIM_Cubic, leaveTangent=0.0247, tangentMode=RCTM_User, time=0.0402, value=1}
          {arriveTangent=-0.3079, interpMode=RCIM_Cubic, leaveTangent=-0.3079, tangentMode=RCTM_User, time=0.3943, value=0.2703}
          {arriveTangent=-0.3302, interpMode=RCIM_Cubic, leaveTangent=-0.3302, tangentMode=RCTM_User, time=1, value=0}
         ]}}}
      Scale Curve=&Multiply_Float(A=&FNTimeOfDayScalar(Night Scale=0.3, Morning Scale=0.7, Evening Scale=0.7), B=1)
     )
    )
    ScaleA=true
    ScaleRGB=true
   )
  ]
  rendererClasses [&NiagaraMeshRendererProperties]
  renderers [{rendererClass=&NiagaraMeshRendererProperties, rendererIndex=0}]
 }
 Flare {
  emitterUpdateScript [
   &EmitterState(Life Cycle Mode=::NewEnumerator1, Loop Behavior=::NewEnumerator1, Loop Delay=0.1, Loop Duration=1, UseLoopDelay=false)
   &SpawnBurst_Instantaneous(Spawn Count=1, Use Spawn Probability=false)
  ]
  particleSpawnScript [
   &InitializeParticle(
    Color=<rgba 5 2.3 0.5 1>
    Lifetime=0.4
    Sprite Rotation Mode=::NewEnumerator2
    Sprite Size Mode=::NewEnumerator3
    Uniform Sprite Size=240
    UsePositionOffset=false
    Write Material Random=true
   )
   &DynamicMaterialParameters(
    Index 0 Param 1=&RandomRangeFloat(Minimum=0.4, Maximum=0.6)
    Index 0 Param 2=1
    Index 0 Param 3=1
    Index 0 Param 4=1
    Module.Param0WriteEnabled=true
    Module.Param1WriteEnabled=true
    Module.Param2WriteEnabled=true
    Module.Param3WriteEnabled=true
   )
  ]
  particleUpdateScript [
   &SkeletalMeshLocation(
    MaskSpawnGroup=false
    Mesh Sampling Type=::NewEnumerator5
    Orientation Sampling=::NewEnumerator1
    Sampled Position Offset=<0 5 0>
    Sampled Position Offset Coordinate Space:ENiagaraSkelMeshTransforms=::NewEnumerator2
    Skeletal Mesh=$System.Skeletal Mesh
    Socket Sampling Mode=::NewEnumerator5
   )
   &ParticleState()
   &SolveForcesAndVelocity(Clamp Velocity=false, Limit Acceleration=false)
   &ScaleColor(
    Scale Alpha=&Multiply_Float(
     A=&FNTimeOfDayScalar(Night Scale=0.3, Morning Scale=0.3, Evening Scale=0.4)
     B=&FloatFromCurve(FloatCurve=DataInterface() {propertyValues {Curve {keys [<at 0 0>,<at 0.2 1>,<at 0.8 1>,<at 1 0>]}}})
    )
    ScaleA=true
    ScaleRGB=true
   )
   &ScaleSpriteSize(Scale Factor=&Vector2DFromFloat(Value=&FloatFromCurve(FloatCurve=DataInterface() {propertyValues {Curve {keys [<at 0 1.5 RCIM_Cubic>,<at 1 0>]}}})))
  ]
  rendererClasses [&NiagaraSpriteRendererProperties]
  renderers [{rendererClass=&NiagaraSpriteRendererProperties}]
 }
 Glints {
  emitterUpdateScript [
   &EmitterState()
   &SpawnRate(SpawnRate=10, Use Spawn Probability=false)
  ]
  particleSpawnScript [
   &InitializeParticle(
    Color=&SelectLinearColorFromArray(
     Array Sampling Mode=::NewEnumerator0
     Direct Array Index=?
     Color Selection Array=DataInterface() {propertyValues {ColorData [<rgba 2 1.04 0.4 1>, <rgba 1.5 1.0625 0.45 1>]}}
     Random Seed=0
    )
    Lifetime=&RandomRangeFloat(Minimum=0.3, Maximum=0.4)
    Position Mode=::NewEnumerator0
    Sprite Rotation Angle=&RandomRangeFloat(Minimum=-10, Maximum=10)
    Sprite Rotation Mode=::NewEnumerator0
    Sprite Size Mode=::NewEnumerator1
    Write Material Random=true
   )
   &PartitionParticles(Number Of Partitions=3) {module=PartitionParticles001}
  ]
  particleUpdateScript [
   &ParticleState()
   &SpriteRotationRate(Rotation Rate=&RandomRangeFloat(Minimum=50, Maximum=200))
   &ScaleSpriteSize(
    Scale Sprite Size Mode=::NewEnumerator0
    Uniform Scale Factor=&FloatFromCurve(
     FloatCurve=DataInterface() {propertyValues {Curve {keys [
         {interpMode=RCIM_Cubic, tangentMode=RCTM_User, time=0, value=1}
         {arriveTangent=-1.9398, interpMode=RCIM_Cubic, leaveTangent=-1.9398, tangentMode=RCTM_User, time=1, value=0}
        ]}}}
     Scale Curve=&FN_LOD_Distance_Remap(Start Distance=100, End Distance=500, Minimum Output=0.75, Maximum Output=1.25, Clamp Output=true)
    )
   )
   &ScaleColor(
    Scale Alpha=&FloatFromCurve(FloatCurve=DataInterface() {propertyValues {Curve {keys [
         {arriveTangent=-2.5249, interpMode=RCIM_Cubic, leaveTangent=-2.5249, tangentMode=RCTM_User, time=0.1408, value=0.9847}
         {arriveTangent=-0.0405, interpMode=RCIM_Cubic, leaveTangent=-0.0405, tangentMode=RCTM_User, time=1, value=0}
        ]}}})
    Scale RGB=&Multiply_VectorByFloat(Vector=<1 1 1>, Float=&FNTimeOfDayScalar(Night Scale=0.8, Morning Scale=0.9, Evening Scale=0.9))
    ScaleA=false
    ScaleRGB=true
   )
   &SkeletalMeshLocation(
    Bone / Socket Index=$Particles.PartitionParticles001.Partition
    MaskSpawnGroup=false
    Mesh Sampling Type=::NewEnumerator5
    Skeletal Mesh=$System.Skeletal Mesh
    Socket Sampling Mode=::NewEnumerator5
   )
   &ShapeLocation(
    Additional Quaternion Rotation=false
    MaskSpawnGroup=true
    Non Uniform Scale=<0.125 0.1 0.005>
    Override Seed=false
    Rotation Mode=::NewEnumerator3
    Rotation Quaternion=$Particles.SkeletalMeshLocation.SampledRotation
    Shape Origin=$Output.SkeletalMeshLocation.SampledPositionWithoutOffset
    Spawn Group:NiagaraInt32=$Particles.PartitionParticles001.Partition
    Sphere Surface Distribution=1
    UseOwnerScale=true
   ) {module=ShapeLocation001}
   &ShapeLocation(
    Additional Quaternion Rotation=false
    MaskSpawnGroup=true
    Non Uniform Scale=<0.08 0.06 0.005>
    Override Seed=false
    Rotation Mode=::NewEnumerator3
    Rotation Quaternion=$Particles.SkeletalMeshLocation.SampledRotation
    Shape Origin=$Output.SkeletalMeshLocation.SampledPositionWithoutOffset
    Spawn Group:NiagaraInt32=$Particles.PartitionParticles001.Partition
    Spawn Group Mask=1
    Sphere Surface Distribution=1
    UseOwnerScale=true
   ) {module=ShapeLocation002}
   &ShapeLocation(
    Additional Quaternion Rotation=false
    MaskSpawnGroup=true
    Non Uniform Scale=<0.08 0.06 0.005>
    Override Seed=false
    Rotation Mode=::NewEnumerator3
    Rotation Quaternion=$Particles.SkeletalMeshLocation.SampledRotation
    Shape Origin=$Output.SkeletalMeshLocation.SampledPositionWithoutOffset
    Spawn Group:NiagaraInt32=$Particles.PartitionParticles001.Partition
    Spawn Group Mask=2
    Sphere Surface Distribution=1
    UseOwnerScale=true
   ) {module=ShapeLocation003}
  ]
  rendererClasses [&NiagaraSpriteRendererProperties]
  renderers [{rendererClass=&NiagaraSpriteRendererProperties}]
 }
 GlowFlare001 {
  emitterUpdateScript [
   &EmitterState(Life Cycle Mode=::NewEnumerator1, Loop Behavior=::NewEnumerator1, Loop Duration=1, UseLoopDelay=false)
   &SpawnBurst_Instantaneous(Spawn Count=1, Use Spawn Probability=false)
  ]
  particleSpawnScript [
   &InitializeParticle(
    Color=<rgba 1 0.6467 0.4271 1.5>
    Lifetime=0.35
    Sprite Size Mode=::NewEnumerator3
    Uniform Sprite Size=200
    UsePositionOffset=false
    Write Material Random=true
   )
   &SkeletalMeshLocation(
    MaskSpawnGroup=false
    Mesh Sampling Type=::NewEnumerator5
    Orientation Sampling=::NewEnumerator1
    Sampled Position Offset Coordinate Space:ENiagaraSkelMeshTransforms=::NewEnumerator2
    Skeletal Mesh=$System.Skeletal Mesh
    Socket Sampling Mode=::NewEnumerator5
   )
  ]
  particleUpdateScript [
   &ParticleState()
   &ScaleColor(
    Scale Alpha=&Multiply_Float(
     A=&FloatFromCurve(FloatCurve=DataInterface() {propertyValues {Curve {keys [{interpMode=RCIM_Cubic, leaveTangent=-0.6434, tangentMode=RCTM_Break, time=0, value=1}, <at 1 0>]}}})
     B=&FNTimeOfDayScalar(Night Scale=0.3, Morning Scale=0.5, Evening Scale=0.5)
    )
    ScaleA=true
    ScaleRGB=true
   )
   &CameraOffset(Camera Offset Amount=50)
   &ScaleSpriteSize(Uniform Curve Sprite Scale=DataInterface() {propertyValues {Curve {keys [{tangentMode=RCTM_User, time=0, value=0}, <at 1 1>]}}})
  ]
  rendererClasses [&NiagaraSpriteRendererProperties]
  renderers [{rendererClass=&NiagaraSpriteRendererProperties}]
 }
 Light {
  emitterUpdateScript [
   &EmitterState(Life Cycle Mode=::NewEnumerator1, Loop Behavior=::NewEnumerator1, Loop Duration=1, UseLoopDelay=false)
   &SpawnBurst_Instantaneous(Spawn Count=1, Use Spawn Probability=false)
  ]
  particleSpawnScript [
   &InitializeParticle(Color=<rgba 5000 2203.1248 338.541 1>, Lifetime=0.4, UsePositionOffset=false, Write Material Random=true)
   &SkeletalMeshLocation(
    MaskSpawnGroup=false
    Mesh Sampling Type=::NewEnumerator5
    Orientation Sampling=::NewEnumerator1
    Sampled Position Offset Coordinate Space:ENiagaraSkelMeshTransforms=::NewEnumerator2
    Skeletal Mesh=$System.Skeletal Mesh
    Socket Sampling Mode=::NewEnumerator5
   )
  ]
  particleUpdateScript [
   &ParticleState()
   &ScaleColor(
    Linear Color Curve=DataInterface() {propertyValues {
      AlphaCurve {keys [<at 0 0>,<at 0.2565 1>,<at 1 0>]}
      BlueCurve {keys [<at 0 1>,<at 0.2541 5>,<at 0.303 1>,<at 1 1>]}
      GreenCurve {keys [<at 0 1>,<at 0.2541 5>,<at 0.303 1>,<at 1 1>]}
      RedCurve {keys [<at 0 1>,<at 0.2541 5>,<at 0.303 1>,<at 1 1>]}
     }}
    Scale Mode=::NewEnumerator2
   )
  ]
  rendererClasses [&NiagaraLightRendererProperties]
  renderers [{rendererClass=&NiagaraLightRendererProperties, rendererIndex=0}]
 }
 Light_Mesh {
  emitterUpdateScript [
   &EmitterState(
    Inactive Response=::NewEnumerator1
    Life Cycle Mode=::NewEnumerator1
    Loop Behavior=::NewEnumerator1
    Loop Duration Mode=::NewEnumerator1
    UseLoopDelay=false
   )
   &SpawnBurst_Instantaneous(Use Spawn Probability=false, UseLoopCountLimit=false)
  ]
  particleSpawnScript [&InitializeParticle(Color=<rgba 1.5 0.78 0.3 1>, Mesh Scale=<1.05 1 1>, Mesh Scale Mode=::NewEnumerator4, UsePositionOffset=false, Write Material Random=true)]
  particleUpdateScript [
   &ParticleState(Kill Particles When Lifetime Has Elapsed=false)
   &SkeletalMeshLocation(Bone Sampling Mode=::NewEnumerator4, MaskSpawnGroup=false, Skeletal Mesh=$System.Skeletal Mesh)
   &ScaleColor(
    Scale Alpha=&FloatFromCurve(
     FloatCurve=DataInterface() {propertyValues {Curve {keys [
         {arriveTangent=-2.6103, interpMode=RCIM_Cubic, leaveTangent=-2.6103, tangentMode=RCTM_User, time=0, value=1}
         {interpMode=RCIM_Cubic, leaveTangent=-4, time=0.4, value=0.2}
        ]}}}
     Scale Curve=&Multiply_Float(A=&FNTimeOfDayScalar(Night Scale=0.5, Morning Scale=0.7, Evening Scale=0.7), B=1)
    )
    Scale RGB=&VectorFromFloat(Value=&FloatFromCurve(FloatCurve=DataInterface() {propertyValues {Curve {keys [
          {arriveTangent=-17.8755, interpMode=RCIM_Cubic, leaveTangent=-17.8755, tangentMode=RCTM_User, time=0, value=5}
          {interpMode=RCIM_Cubic, leaveTangent=-4, time=0.4, value=1}
         ]}}}))
    ScaleA=true
    ScaleRGB=true
   ) {module=ScaleColor001}
  ]
  rendererClasses [&NiagaraMeshRendererProperties]
  renderers [{rendererClass=&NiagaraMeshRendererProperties, rendererIndex=0}]
 }
 Sheen_Mesh {
  emitterUpdateScript [
   &EmitterState(
    Inactive Response=::NewEnumerator1
    Life Cycle Mode=::NewEnumerator1
    Loop Duration=&RandomRangeFloat(Minimum=3, Maximum=6)
    UseLoopDelay=false
   )
   &SpawnBurst_Instantaneous(Use Spawn Probability=false, UseLoopCountLimit=false)
  ]
  particleSpawnScript [
   &InitializeParticle(
    Color=<rgba 10 5.206 2.01 2>
    Lifetime=&RandomRangeFloat(Minimum=0.5, Maximum=0.666)
    Mesh Scale=<1.05 1 1>
    Mesh Scale Mode=::NewEnumerator4
    UsePositionOffset=false
    Write Material Random=true
   )
   &DynamicMaterialParameters(
    Index 0 Param 2=0.2
    Index 0 Param 3=-0.1
    Module.Param0WriteEnabled=false
    Module.Param1WriteEnabled=true
    Module.Param2WriteEnabled=true
    Module.Param3WriteEnabled=false
   )
  ]
  particleUpdateScript [
   &ParticleState()
   &DynamicMaterialParameters(
    Index 0 Param 1=&FloatFromCurve(FloatCurve=DataInterface() {propertyValues {Curve {keys [{interpMode=RCIM_Cubic, tangentMode=RCTM_User, time=0, value=1}, {interpMode=RCIM_Cubic, tangentMode=RCTM_User, time=1, value=0}]}}})
    Index 0 Param 2=0.5
    Module.Param0WriteEnabled=true
    Module.Param1WriteEnabled=true
    Module.Param2WriteEnabled=false
    Module.Param3WriteEnabled=false
   ) {module=DynamicMaterialParameters001}
   &SkeletalMeshLocation(Bone Sampling Mode=::NewEnumerator4, MaskSpawnGroup=false, Skeletal Mesh=$System.Skeletal Mesh)
  ]
  rendererClasses [&NiagaraMeshRendererProperties]
  renderers [{rendererClass=&NiagaraMeshRendererProperties, rendererIndex=0}]
 }
 Sprite_Gold001 {
  simTarget=GPUComputeSim
  emitterUpdateScript [
   &EmitterState(
    Enable Distance Culling=true
    Enable Visibility Culling=true
    Life Cycle Mode=::NewEnumerator1
    Loop Behavior=::NewEnumerator1
    Loop Duration=2
    MaxDistance=2500
    Module.bApplySpawnCountScaleByDistance=true
    Module.bUseMaxDistance=true
    Module.bUseMinDistance=false
    Scalability Mode=::NewEnumerator1
    Scale Spawn Count=true
    UseLoopDelay=false
   )
   &SpawnBurst_Instantaneous(
    Spawn Count=&RandomRangeInt(Minimum=17, Maximum=25)
    Use Spawn Probability=false
    UseLoopCountLimit=false
   )
  ]
  particleSpawnScript [
   &InitializeParticle(
    Color=<rgba 16 7.925 2.4 1>
    Lifetime Max=1.2
    Lifetime Min=0.3
    Lifetime Mode=::NewEnumerator1
    Position Offset=&RandomRangeVector(Minimum=<35 25 40>, Maximum=<-35 -25 40>)
    Sprite Size Mode=::NewEnumerator1
    Uniform Sprite Size Max=2
    Uniform Sprite Size Min=0.5
    UsePositionOffset=true
    Write Material Random=true
   )
   &SkeletalMeshLocation(
    "Inherit Velocity (Scale)"=0
    MaskSpawnGroup=false
    Mesh Sampling Type=::NewEnumerator5
    Orientation Sampling=::NewEnumerator1
    Skeletal Mesh=$User.Skeletal_Mesh
    Socket Sampling Mode=::NewEnumerator5
    Velocity Sampling=::NewEnumerator0
    VelocitySpeedLimit=false
   )
   &ShapeLocation(
    Additional Quaternion Rotation=false
    MaskSpawnGroup=false
    Non Uniform Scale=<0.5 0.5 0.2>
    Offset=<0 0 -1>
    Offset Mode=::NewEnumerator0
    Override Seed=false
    Rotation Mode=::NewEnumerator3
    Rotation Quaternion=$Particles.SkeletalMeshLocation.SampledRotation
    Sphere Radius=20
    Transform Order:ENiagara_TransformOrder=::NewEnumerator1
    UseOwnerScale=true
   ) {module=ShapeLocation001}
   &AddVelocity(
    Module.Use Falloff Distance=false
    Override Seed=false
    Velocity Mode=::NewEnumerator1
    Velocity Origin=$Output.SkeletalMeshLocation.SampledPositionWithoutOffset
    Velocity Speed=&RandomRangeFloat(Minimum=10, Maximum=400)
   ) {module=AddVelocity001}
  ]
  particleUpdateScript [
   &ParticleState()
   &AccelerationForce(Acceleration=<0 0 -10>)
   &Drag(Drag=4, Use Linear Drag=true, Use Rotational Drag=true)
   &SolveForcesAndVelocity(Clamp Velocity=false, Limit Acceleration=false)
   &ScaleSpriteSize(Uniform Curve Sprite Scale=DataInterface() {propertyValues {Curve {keys [
        {arriveTangent=9.0725, interpMode=RCIM_Cubic, leaveTangent=3.7536, tangentMode=RCTM_Break, time=0, value=0}
        {arriveTangent=1.0728, interpMode=RCIM_Cubic, tangentMode=RCTM_Break, time=0.2, value=0.5}
        {arriveTangent=-0.625, interpMode=RCIM_Cubic, leaveTangent=-0.625, time=0.8, value=0.9847}
        {arriveTangent=-0.0405, interpMode=RCIM_Cubic, leaveTangent=-0.0405, tangentMode=RCTM_User, time=1, value=0}
       ]}}})
  ]
  rendererClasses [&NiagaraSpriteRendererProperties]
  renderers [{rendererClass=&NiagaraSpriteRendererProperties}]
 }
 Sprite_Gold002 {
  simTarget=GPUComputeSim
  emitterUpdateScript [
   &EmitterState(
    Enable Distance Culling=true
    Enable Visibility Culling=true
    Life Cycle Mode=::NewEnumerator1
    Loop Duration=2
    MaxDistance=2500
    Module.bApplySpawnCountScaleByDistance=true
    Module.bUseMaxDistance=true
    Module.bUseMinDistance=false
    Scalability Mode=::NewEnumerator1
    Scale Spawn Count=true
    UseLoopDelay=false
   )
   &SpawnBurst_Instantaneous(Spawn Count=3, Use Spawn Probability=false, UseLoopCountLimit=false)
   &SpawnRate(SpawnRate=4, Use Spawn Probability=false)
  ]
  particleSpawnScript [
   &InitializeParticle(
    Color=<rgba 16 7.925 2.4 1>
    Lifetime Min=0.6
    Lifetime Mode=::NewEnumerator1
    Position Offset=&RandomRangeVector(Minimum=<35 25 40>, Maximum=<-35 -25 40>)
    Sprite Size Mode=::NewEnumerator1
    Uniform Sprite Size Max=2
    Uniform Sprite Size Min=1
    UsePositionOffset=true
    Write Material Random=true
   )
   &SkeletalMeshLocation(
    MaskSpawnGroup=false
    Mesh Sampling Type=::NewEnumerator5
    Orientation Sampling=::NewEnumerator1
    Skeletal Mesh=$System.Skeletal Mesh
    Socket Sampling Mode=::NewEnumerator5
    Velocity Sampling=::NewEnumerator0
    VelocitySpeedLimit=false
   )
   &ShapeLocation(
    Additional Quaternion Rotation=false
    MaskSpawnGroup=false
    Non Uniform Scale=<1 0.5 0.15>
    Offset=<0 0 -1>
    Offset Mode=::NewEnumerator0
    Override Seed=false
    Rotation Mode=::NewEnumerator3
    Rotation Quaternion=$Particles.SkeletalMeshLocation.SampledRotation
    Sphere Radius=20
    Transform Order:ENiagara_TransformOrder=::NewEnumerator1
    UseOwnerScale=true
   ) {module=ShapeLocation001}
   &AddVelocity(Velocity=&RandomRangeVector(Minimum=<50 50 0>, Maximum=<-50 0 0>))
  ]
  particleUpdateScript [
   &ParticleState()
   &AccelerationForce(Acceleration=<0 0 -20>)
   &Drag(Use Linear Drag=true, Use Rotational Drag=true)
   &SolveForcesAndVelocity(Clamp Velocity=false, Limit Acceleration=false)
   &ScaleSpriteSize(Uniform Curve Sprite Scale=DataInterface() {propertyValues {Curve {keys [
        {arriveTangent=9.0725, interpMode=RCIM_Cubic, leaveTangent=3.7536, tangentMode=RCTM_Break, time=0, value=0}
        {arriveTangent=1.0728, interpMode=RCIM_Cubic, tangentMode=RCTM_Break, time=0.2, value=0.5}
        {arriveTangent=-0.625, interpMode=RCIM_Cubic, leaveTangent=-0.625, time=0.8, value=0.9847}
        {arriveTangent=-0.0405, interpMode=RCIM_Cubic, leaveTangent=-0.0405, tangentMode=RCTM_User, time=1, value=0}
       ]}}})
  ]
  rendererClasses [&NiagaraSpriteRendererProperties]
  renderers [{rendererClass=&NiagaraSpriteRendererProperties}]
 }
}
summary {
 systemName=NS_Emote_RollerBlade
 emitters [
  {emitterName=Glints, rendererClasses [&NiagaraSpriteRendererProperties]}
  {emitterName=Sheen_Mesh, rendererClasses [&NiagaraMeshRendererProperties]}
  {emitterName=Corona, rendererClasses [&NiagaraMeshRendererProperties]}
  {emitterName=Light_Mesh, rendererClasses [&NiagaraMeshRendererProperties]}
  {emitterName=BG_Flare, rendererClasses [&NiagaraSpriteRendererProperties]}
  {emitterName=Light, rendererClasses [&NiagaraLightRendererProperties]}
  {emitterName=GlowFlare001, rendererClasses [&NiagaraSpriteRendererProperties]}
  {emitterName=Flare, rendererClasses [&NiagaraSpriteRendererProperties]}
  {emitterName=Sprite_Gold001, simTarget=GPUComputeSim, rendererClasses [&NiagaraSpriteRendererProperties]}
  {emitterName=Sprite_Gold002, simTarget=GPUComputeSim, rendererClasses [&NiagaraSpriteRendererProperties]}
 ]
 userVariables [
  User.bIsFrontEnd {type=NiagaraBool}
  User.bIsFrontEndPreview {type=NiagaraBool}
  User.Skeletal_Mesh {
   type=NiagaraDataInterfaceSkeletalMesh
   defaultValue {
    struct=NiagaraExt_VariableValue_DataInterface
    value {dataInterface=&self:NiagaraDataInterfaceSkeletalMesh_0}
   }
  }
 ]
}
